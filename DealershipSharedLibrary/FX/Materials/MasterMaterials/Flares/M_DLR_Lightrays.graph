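# M_DLR_Lightrays is an unlit additive particle material drawing animated radial light rays, with optional hue-shifting for disco lights.
# Counter-scrolling, particle-randomized polar samples of a line texture multiply together, then take contrast and noise masking.
# A center glow, optional Ceil-stepped style and noise-varied radial falloff shape the result.
# Particle color or parameter colors tint it; depth fade, with mobile/quality bypasses, and optional near-camera fade control opacity.
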Material /DealershipSharedLibrary/FX/Materials/MasterMaterials/Flares/M_DLR_Lightrays
graph {
 expressions {
  MaterialExpressionAdd_1 {
   class=MaterialExpressionAdd
   inputs [
    A {from=MaterialExpressionNamedRerouteUsage_3}
    B {from=MaterialExpressionMultiply_25}
   ]
  }
  MaterialExpressionAdd_15 {
   class=MaterialExpressionAdd
   inputs [
    A {from=MaterialExpressionParticleRandom_4}
    B {from=MaterialExpressionTime_7}
   ]
  }
  MaterialExpressionAdd_16 {
   class=MaterialExpressionAdd
   inputs [
    A {from=MaterialExpressionParticleRandom_5}
    B {from=MaterialExpressionTime_8}
   ]
  }
  MaterialExpressionAdd_2 {
   class=MaterialExpressionAdd
   inputs [
    A {from=MaterialExpressionParticleRandom_0}
    B {from=MaterialExpressionTime_0}
   ]
  }
  MaterialExpressionAdd_4 {
   class=MaterialExpressionAdd
   inputs [
    A {from=MaterialExpressionAdd_1}
    B {from=MaterialExpressionScalarParameter_15}
   ]
  }
  MaterialExpressionAdd_6 {
   class=MaterialExpressionAdd
   inputs [
    A {from=MaterialExpressionNamedRerouteUsage_0}
    B {from=MaterialExpressionScalarParameter_21}
   ]
  }
  MaterialExpressionAdd_8 {
   class=MaterialExpressionAdd
   inputs [
    A {from=MaterialExpressionParticleRandom_1}
    B {from=MaterialExpressionMultiply_38}
   ]
  }
  MaterialExpressionCeil_1 {class=MaterialExpressionCeil, inputs [None {from=MaterialExpressionAdd_4, output=""}], properties {}}
  MaterialExpressionComponentMask_2 {class=MaterialExpressionComponentMask, inputs [None {from=MaterialExpressionAdd_6}], properties {g=true}}
  MaterialExpressionComponentMask_3 {class=MaterialExpressionComponentMask, inputs [None {from=MaterialExpressionNamedRerouteDeclaration_0}], properties {g=true}}
  MaterialExpressionConstant_3 {class=MaterialExpressionConstant, properties {r=1}}
  MaterialExpressionConstant_4 {class=MaterialExpressionConstant, properties {r=1}}
  MaterialExpressionConvert_1 {
   class=MaterialExpressionConvert
   inputs [
    None {from=MaterialExpressionScalarParameter_27}
    None {from=""}
   ]
   properties {
    nodeName=Make Float2
    convertInputs [
     {
      type=Scalar
      expressionInput {
       expression=&self:MaterialExpressionScalarParameter_27
       inputName=None
       mask=0
       maskA=0
       maskB=0
       maskG=0
       maskR=0
       outputIndex=0
      }
     }
     {
      type=Scalar
      defaultValue=<rgba 1 0 0 1>
      expressionInput {expression=null, inputName=None, mask=0, maskA=0, maskB=0, maskG=0, maskR=0, outputIndex=0}
     }
    ]
    convertOutputs [{type=Vector2}]
   }
  }
  MaterialExpressionConvert_4 {
   class=MaterialExpressionConvert
   inputs [
    None {from=MaterialExpressionScalarParameter_48}
    None {from=""}
   ]
   properties {
    nodeName=Make Float2
    convertInputs [
     {
      type=Scalar
      expressionInput {
       expression=&self:MaterialExpressionScalarParameter_48
       inputName=None
       mask=0
       maskA=0
       maskB=0
       maskG=0
       maskR=0
       outputIndex=0
      }
     }
     {
      type=Scalar
      defaultValue=<rgba 1 0 0 1>
      expressionInput {expression=null, inputName=None, mask=0, maskA=0, maskB=0, maskG=0, maskR=0, outputIndex=0}
     }
    ]
    convertOutputs [{type=Vector2}]
   }
  }
  MaterialExpressionConvert_6 {
   class=MaterialExpressionConvert
   inputs [
    None {from=""}
    None {from=""}
   ]
   properties {
    nodeName=Make Float2
    convertInputs [
     {
      type=Scalar
      defaultValue=<rgba 3 0 0 1>
      expressionInput {expression=null, inputName=None, mask=0, maskA=0, maskB=0, maskG=0, maskR=0, outputIndex=0}
     }
     {type=Scalar, expressionInput {expression=null, inputName=None, mask=0, maskA=0, maskB=0, maskG=0, maskR=0, outputIndex=0}}
    ]
    convertOutputs [{type=Vector2}]
   }
  }
  MaterialExpressionConvert_7 {
   class=MaterialExpressionConvert
   inputs [
    None {from=""}
    None {from=""}
   ]
   properties {
    nodeName=Make Float2
    convertInputs [
     {
      type=Scalar
      defaultValue=<rgba 0.33 0 0 1>
      expressionInput {expression=null, inputName=None, mask=0, maskA=0, maskB=0, maskG=0, maskR=0, outputIndex=0}
     }
     {
      type=Scalar
      defaultValue=<rgba 0.01 0 0 1>
      expressionInput {expression=null, inputName=None, mask=0, maskA=0, maskB=0, maskG=0, maskR=0, outputIndex=0}
     }
    ]
    convertOutputs [{type=Vector2}]
   }
  }
  MaterialExpressionDepthFade_1 {
   class=MaterialExpressionDepthFade
   inputs [
    Opacity {from="", output=""}
    FadeDistance {from=MaterialExpressionScalarParameter_14, output=""}
   ]
   properties {}
  }
  MaterialExpressionFeatureLevelSwitch_1 {
   class=MaterialExpressionFeatureLevelSwitch
   inputs [
    Default {from=MaterialExpressionReroute_9, output=""}
    ES2 {from=MaterialExpressionConstant_4, output=""}
    ES3_1 {from=MaterialExpressionConstant_4, output=""}
    SM4_REMOVED {from=MaterialExpressionReroute_16, output=""}
    SM5 {from="", output=""}
    SM6 {from="", output=""}
   ]
   properties {}
  }
  MaterialExpressionFrac_1 {class=MaterialExpressionFrac, inputs [None {from=MaterialExpressionMultiply_30, output=""}], properties {}}
  MaterialExpressionLinearInterpolate_4 {
   class=MaterialExpressionLinearInterpolate
   inputs [
    A {from=MaterialExpressionAdd_1}
    B {from=MaterialExpressionMultiply_29}
    Alpha {from=MaterialExpressionScalarParameter_20}
   ]
  }
  MaterialExpressionLinearInterpolate_5 {
   class=MaterialExpressionLinearInterpolate
   inputs [
    A {from=""}
    B {from=""}
    Alpha {from=MaterialExpressionTextureSample_7, output=R}
   ]
   properties {constA=2, constB=6}
  }
  MaterialExpressionLinearInterpolate_6 {
   class=MaterialExpressionLinearInterpolate
   inputs [
    A {from=MaterialExpressionVectorParameter_2, output=RGB}
    B {from=MaterialExpressionVectorParameter_3, output=RGB}
    Alpha {from=""}
   ]
  }
  MaterialExpressionLinearInterpolate_7 {
   class=MaterialExpressionLinearInterpolate
   inputs [
    A {from=MaterialExpressionReroute_8}
    B {from=MaterialExpressionMultiply_41}
    Alpha {from=MaterialExpressionSaturate_1}
   ]
  }
  MaterialExpressionMaterialFunctionCall_12 {
   class=MaterialExpressionMaterialFunctionCall
   inputs [
    Vector or UVs {from=MaterialExpressionMultiply_42}
    Phase Offset {from=MaterialExpressionMultiply_35}
    Swizzle Coordinate Output {from=""}
   ]
   properties {materialFunction=/DealershipSharedLibrary/FX/Materials/MaterialFunctions/MF_VectorToRadialValue_Phase}
  }
  MaterialExpressionMaterialFunctionCall_13 {
   class=MaterialExpressionMaterialFunctionCall
   inputs [
    Vector or UVs {from=MaterialExpressionMultiply_45}
    Phase Offset {from=MaterialExpressionMultiply_36}
    Swizzle Coordinate Output {from=""}
   ]
   properties {materialFunction=&MF_VectorToRadialValue_Phase}
  }
  MaterialExpressionMaterialFunctionCall_25 {
   class=MaterialExpressionMaterialFunctionCall
   inputs [
    Vector or UVs {from=MaterialExpressionMultiply_68}
    Phase Offset {from=MaterialExpressionAdd_8}
    Swizzle Coordinate Output {from=""}
   ]
   properties {materialFunction=&MF_VectorToRadialValue_Phase}
  }
  MaterialExpressionMaterialFunctionCall_26 {
   class=MaterialExpressionMaterialFunctionCall
   inputs [
    Vector or UVs {from=MaterialExpressionMultiply_70}
    Phase Offset {from=MaterialExpressionMultiply_39}
    Swizzle Coordinate Output {from=""}
   ]
   properties {materialFunction=&MF_VectorToRadialValue_Phase}
  }
  MaterialExpressionMaterialFunctionCall_27 {
   class=MaterialExpressionMaterialFunctionCall
   inputs [
    Vector or UVs {from=MaterialExpressionMultiply_72}
    Phase Offset {from=""}
    Swizzle Coordinate Output {from=""}
   ]
   properties {materialFunction=&MF_VectorToRadialValue_Phase}
  }
  MaterialExpressionMaterialFunctionCall_7 {
   class=MaterialExpressionMaterialFunctionCall
   inputs [Emissive {from=MaterialExpressionNamedRerouteUsage_6}]
   properties {materialFunction=&MF_FX_EmissiveAdjustments}
  }
  MaterialExpressionMaterialFunctionCall_8 {
   class=MaterialExpressionMaterialFunctionCall
   inputs [
    Hue Shift Percentage {from=MaterialExpressionStaticSwitchParameter_13}
    Texture {from=MaterialExpressionMultiply_20}
   ]
   properties {materialFunction=&HueShift}
  }
  MaterialExpressionMaterialFunctionCall_9 {
   class=MaterialExpressionMaterialFunctionCall
   inputs [Distance {from=MaterialExpressionScalarParameter_25}]
   properties {materialFunction=&Near_Camera_Fade}
  }
  MaterialExpressionMultiply_19 {
   class=MaterialExpressionMultiply
   inputs [
    A {from=MaterialExpressionLinearInterpolate_7}
    B {from=MaterialExpressionReroute_20}
   ]
  }
  MaterialExpressionMultiply_20 {
   class=MaterialExpressionMultiply
   inputs [
    A {from=MaterialExpressionStaticSwitchParameter_12}
    B {from=MaterialExpressionNamedRerouteUsage_4}
   ]
  }
  MaterialExpressionMultiply_21 {
   class=MaterialExpressionMultiply
   inputs [
    A {from=MaterialExpressionReroute_24}
    B {from=MaterialExpressionTextureSample_2, output=G}
   ]
  }
  MaterialExpressionMultiply_22 {
   class=MaterialExpressionMultiply
   inputs [
    A {from=MaterialExpressionReroute_21}
    B {from=MaterialExpressionPower_3}
   ]
  }
  MaterialExpressionMultiply_25 {
   class=MaterialExpressionMultiply
   inputs [
    A {from=MaterialExpressionPower_0}
    B {from=MaterialExpressionScalarParameter_13}
   ]
  }
  MaterialExpressionMultiply_26 {
   class=MaterialExpressionMultiply
   inputs [
    A {from=MaterialExpressionScalarParameter_26}
    B {from=MaterialExpressionAdd_2}
   ]
  }
  MaterialExpressionMultiply_27 {
   class=MaterialExpressionMultiply
   inputs [
    A {from=MaterialExpressionReroute_26}
    B {from=MaterialExpressionMultiply_28}
   ]
  }
  MaterialExpressionMultiply_28 {
   class=MaterialExpressionMultiply
   inputs [
    A {from=MaterialExpressionTextureSample_6, output=R}
    B {from=MaterialExpressionScalarParameter_18}
   ]
  }
  MaterialExpressionMultiply_29 {
   class=MaterialExpressionMultiply
   inputs [
    A {from=MaterialExpressionCeil_1}
    B {from=MaterialExpressionScalarParameter_19}
   ]
  }
  MaterialExpressionMultiply_30 {
   class=MaterialExpressionMultiply
   inputs [
    A {from=MaterialExpressionTime_2}
    B {from=MaterialExpressionScalarParameter_22}
   ]
   properties {constB=1.2}
  }
  MaterialExpressionMultiply_31 {
   class=MaterialExpressionMultiply
   inputs [
    A {from=MaterialExpressionLinearInterpolate_6}
    B {from=MaterialExpressionParticleColor_1, output=RGB}
   ]
  }
  MaterialExpressionMultiply_32 {
   class=MaterialExpressionMultiply
   inputs [
    A {from=MaterialExpressionFrac_1}
    B {from=MaterialExpressionScalarParameter_24}
   ]
  }
  MaterialExpressionMultiply_33 {
   class=MaterialExpressionMultiply
   inputs [
    A {from=MaterialExpressionMaterialFunctionCall_9, output=Result}
    B {from=MaterialExpressionMultiply_19}
   ]
  }
  MaterialExpressionMultiply_34 {
   class=MaterialExpressionMultiply
   inputs [
    A {from=MaterialExpressionConvert_1}
    B {from=""}
   ]
   properties {constA=2, constB=2}
  }
  MaterialExpressionMultiply_35 {
   class=MaterialExpressionMultiply
   inputs [
    A {from=MaterialExpressionMultiply_75}
    B {from=""}
   ]
   properties {constB=0.07}
  }
  MaterialExpressionMultiply_36 {
   class=MaterialExpressionMultiply
   inputs [
    A {from=MaterialExpressionMultiply_74}
    B {from=""}
   ]
   properties {constB=-0.047}
  }
  MaterialExpressionMultiply_37 {
   class=MaterialExpressionMultiply
   inputs [
    A {from=MaterialExpressionMultiply_76}
    B {from=""}
   ]
   properties {constA=2, constB=2.5}
  }
  MaterialExpressionMultiply_38 {
   class=MaterialExpressionMultiply
   inputs [
    A {from=MaterialExpressionMultiply_26}
    B {from=""}
   ]
   properties {constB=0.044}
  }
  MaterialExpressionMultiply_39 {
   class=MaterialExpressionMultiply
   inputs [
    A {from=MaterialExpressionTime_3}
    B {from=""}
   ]
   properties {constB=0.2}
  }
  MaterialExpressionMultiply_40 {
   class=MaterialExpressionMultiply
   inputs [
    A {from=MaterialExpressionStaticSwitchParameter_10}
    B {from=MaterialExpressionScalarParameter_28}
   ]
  }
  MaterialExpressionMultiply_41 {
   class=MaterialExpressionMultiply
   inputs [
    A {from=MaterialExpressionParticleColor_3, output=A}
    B {from=MaterialExpressionNamedRerouteUsage_5}
   ]
  }
  MaterialExpressionMultiply_42 {
   class=MaterialExpressionMultiply
   inputs [
    A {from=MaterialExpressionSubtract_2}
    B {from=""}
   ]
   properties {constB=2}
  }
  MaterialExpressionMultiply_43 {
   class=MaterialExpressionMultiply
   inputs [
    A {from=MaterialExpressionMaterialFunctionCall_12, output=Radial Coordinates}
    B {from=MaterialExpressionMultiply_34}
   ]
  }
  MaterialExpressionMultiply_45 {
   class=MaterialExpressionMultiply
   inputs [
    A {from=MaterialExpressionSubtract_3}
    B {from=""}
   ]
   properties {constB=2}
  }
  MaterialExpressionMultiply_46 {
   class=MaterialExpressionMultiply
   inputs [
    A {from=MaterialExpressionMaterialFunctionCall_13, output=Radial Coordinates}
    B {from=MaterialExpressionMultiply_37}
   ]
  }
  MaterialExpressionMultiply_68 {
   class=MaterialExpressionMultiply
   inputs [
    A {from=MaterialExpressionSubtract_4}
    B {from=""}
   ]
   properties {constB=2}
  }
  MaterialExpressionMultiply_69 {
   class=MaterialExpressionMultiply
   inputs [
    A {from=MaterialExpressionMaterialFunctionCall_25, output=Radial Coordinates}
    B {from=MaterialExpressionConvert_7}
   ]
  }
  MaterialExpressionMultiply_70 {
   class=MaterialExpressionMultiply
   inputs [
    A {from=MaterialExpressionSubtract_5}
    B {from=""}
   ]
   properties {constB=2}
  }
  MaterialExpressionMultiply_71 {
   class=MaterialExpressionMultiply
   inputs [
    A {from=MaterialExpressionMaterialFunctionCall_26, output=Radial Coordinates}
    B {from=MaterialExpressionConvert_6}
   ]
  }
  MaterialExpressionMultiply_72 {
   class=MaterialExpressionMultiply
   inputs [
    A {from=MaterialExpressionSubtract_6}
    B {from=""}
   ]
   properties {constB=2}
  }
  MaterialExpressionMultiply_73 {
   class=MaterialExpressionMultiply
   inputs [
    A {from=MaterialExpressionMaterialFunctionCall_27, output=Radial Coordinates}
    B {from=""}
   ]
  }
  MaterialExpressionMultiply_74 {
   class=MaterialExpressionMultiply
   inputs [
    A {from=MaterialExpressionScalarParameter_46}
    B {from=MaterialExpressionAdd_15}
   ]
  }
  MaterialExpressionMultiply_75 {
   class=MaterialExpressionMultiply
   inputs [
    A {from=MaterialExpressionScalarParameter_47}
    B {from=MaterialExpressionAdd_16}
   ]
  }
  MaterialExpressionMultiply_76 {
   class=MaterialExpressionMultiply
   inputs [
    A {from=MaterialExpressionConvert_4}
    B {from=""}
   ]
   properties {constA=2, constB=2}
  }
  MaterialExpressionNamedRerouteDeclaration_0 {
   class=MaterialExpressionNamedRerouteDeclaration
   inputs [None {from=MaterialExpressionMultiply_73}]
   properties {name=RadialUV, nodeColor=<rgba 0.1016 0 1 1>}
  }
  MaterialExpressionNamedRerouteDeclaration_3 {
   class=MaterialExpressionNamedRerouteDeclaration
   inputs [None {from=MaterialExpressionMultiply_27}]
   properties {name=Radial Texture Out, nodeColor=<rgba 0.5071 0 1 1>}
  }
  MaterialExpressionNamedRerouteDeclaration_4 {
   class=MaterialExpressionNamedRerouteDeclaration
   inputs [None {from=MaterialExpressionMultiply_22}]
   properties {name=Full Texture Out, nodeColor=<rgba 1 0 0.0703 1>}
  }
  MaterialExpressionNamedRerouteDeclaration_5 {
   class=MaterialExpressionNamedRerouteDeclaration
   inputs [None {from=MaterialExpressionMultiply_40}]
   properties {name=Emissive Out, nodeColor=<rgba 0 0.3906 1 1>}
  }
  MaterialExpressionNamedRerouteDeclaration_6 {
   class=MaterialExpressionNamedRerouteDeclaration
   inputs [None {from=MaterialExpressionStaticSwitchParameter_9}]
   properties {name=Opacity Out, nodeColor=<rgba 0.7109 1 0 1>}
  }
  MaterialExpressionNamedRerouteUsage_0 {class=MaterialExpressionNamedRerouteUsage}
  MaterialExpressionNamedRerouteUsage_3 {class=MaterialExpressionNamedRerouteUsage}
  MaterialExpressionNamedRerouteUsage_4 {class=MaterialExpressionNamedRerouteUsage}
  MaterialExpressionNamedRerouteUsage_5 {class=MaterialExpressionNamedRerouteUsage}
  MaterialExpressionNamedRerouteUsage_6 {class=MaterialExpressionNamedRerouteUsage}
  MaterialExpressionNamedRerouteUsage_7 {class=MaterialExpressionNamedRerouteUsage}
  MaterialExpressionOneMinus_2 {class=MaterialExpressionOneMinus, inputs [None {from=MaterialExpressionComponentMask_2}]}
  MaterialExpressionOneMinus_3 {class=MaterialExpressionOneMinus, inputs [None {from=MaterialExpressionComponentMask_3}]}
  MaterialExpressionParticleColor_1 {class=MaterialExpressionParticleColor, inputs [], properties {}}
  MaterialExpressionParticleColor_3 {class=MaterialExpressionParticleColor, inputs [], properties {}}
  MaterialExpressionParticleRandom_0 {class=MaterialExpressionParticleRandom, inputs [], properties {}}
  MaterialExpressionParticleRandom_1 {class=MaterialExpressionParticleRandom, inputs [], properties {}}
  MaterialExpressionParticleRandom_4 {class=MaterialExpressionParticleRandom, inputs [], properties {}}
  MaterialExpressionParticleRandom_5 {class=MaterialExpressionParticleRandom, inputs [], properties {}}
  MaterialExpressionPower_0 {
   class=MaterialExpressionPower
   inputs [
    Base {from=MaterialExpressionOneMinus_2}
    Exp {from=MaterialExpressionScalarParameter_12}
   ]
  }
  MaterialExpressionPower_1 {
   class=MaterialExpressionPower
   inputs [
    Base {from=MaterialExpressionReroute_27}
    Exp {from=MaterialExpressionScalarParameter_0}
   ]
  }
  MaterialExpressionPower_3 {
   class=MaterialExpressionPower
   inputs [
    Base {from=MaterialExpressionOneMinus_3}
    Exp {from=MaterialExpressionLinearInterpolate_5}
   ]
  }
  MaterialExpressionQualitySwitch_0 {
   class=MaterialExpressionQualitySwitch
   inputs [
    Default {from=MaterialExpressionReroute_11, output=""}
    Low {from=MaterialExpressionConstant_3, output=""}
    High {from=MaterialExpressionReroute_15, output=""}
    Medium {from=MaterialExpressionReroute_14, output=""}
    Epic {from="", output=""}
   ]
   properties {}
  }
  MaterialExpressionReroute_10 {class=MaterialExpressionReroute, inputs [None {from=MaterialExpressionDepthFade_1}]}
  MaterialExpressionReroute_11 {class=MaterialExpressionReroute, inputs [None {from=MaterialExpressionDepthFade_1}]}
  MaterialExpressionReroute_12 {class=MaterialExpressionReroute, inputs [None {from=MaterialExpressionDepthFade_1}]}
  MaterialExpressionReroute_13 {class=MaterialExpressionReroute, inputs [None {from=MaterialExpressionDepthFade_1}]}
  MaterialExpressionReroute_14 {class=MaterialExpressionReroute, inputs [None {from=MaterialExpressionReroute_12}]}
  MaterialExpressionReroute_15 {class=MaterialExpressionReroute, inputs [None {from=MaterialExpressionReroute_13}]}
  MaterialExpressionReroute_16 {class=MaterialExpressionReroute, inputs [None {from=MaterialExpressionReroute_10}]}
  MaterialExpressionReroute_17 {class=MaterialExpressionReroute, inputs [None {from=MaterialExpressionReroute_18}]}
  MaterialExpressionReroute_18 {class=MaterialExpressionReroute, inputs [None {from=MaterialExpressionMultiply_19}]}
  MaterialExpressionReroute_19 {class=MaterialExpressionReroute, inputs [None {from=MaterialExpressionStaticSwitchParameter_11}]}
  MaterialExpressionReroute_20 {class=MaterialExpressionReroute, inputs [None {from=MaterialExpressionReroute_19}]}
  MaterialExpressionReroute_21 {class=MaterialExpressionReroute, inputs [None {from=MaterialExpressionReroute_7}]}
  MaterialExpressionReroute_22 {class=MaterialExpressionReroute, inputs [None {from=MaterialExpressionReroute_23}]}
  MaterialExpressionReroute_23 {class=MaterialExpressionReroute, inputs [None {from=MaterialExpressionParticleColor_1, output=RGB}]}
  MaterialExpressionReroute_24 {class=MaterialExpressionReroute, inputs [None {from=MaterialExpressionReroute_25}]}
  MaterialExpressionReroute_25 {class=MaterialExpressionReroute, inputs [None {from=MaterialExpressionTextureSample_3, output=R}]}
  MaterialExpressionReroute_26 {class=MaterialExpressionReroute, inputs [None {from=MaterialExpressionPower_1}]}
  MaterialExpressionReroute_27 {class=MaterialExpressionReroute, inputs [None {from=MaterialExpressionMultiply_21}]}
  MaterialExpressionReroute_7 {class=MaterialExpressionReroute, inputs [None {from=MaterialExpressionLinearInterpolate_4}]}
  MaterialExpressionReroute_8 {class=MaterialExpressionReroute, inputs [None {from=MaterialExpressionParticleColor_3, output=A}]}
  MaterialExpressionReroute_9 {class=MaterialExpressionReroute, inputs [None {from=MaterialExpressionDepthFade_1}]}
  MaterialExpressionSaturate_1 {class=MaterialExpressionSaturate, inputs [None {from=MaterialExpressionScalarParameter_29}]}
  MaterialExpressionScalarParameter_0 {class=MaterialExpressionScalarParameter, properties {defaultValue=1, parameterName=Contrast}}
  MaterialExpressionScalarParameter_12 {class=MaterialExpressionScalarParameter, properties {defaultValue=10, parameterName=Center Exponent}}
  MaterialExpressionScalarParameter_13 {class=MaterialExpressionScalarParameter, properties {defaultValue=1, parameterName=Center Brightness}}
  MaterialExpressionScalarParameter_14 {class=MaterialExpressionScalarParameter, properties {defaultValue=500, parameterName=Depth Fade Distance}}
  MaterialExpressionScalarParameter_15 {class=MaterialExpressionScalarParameter, properties {defaultValue=-0.7, parameterName=Ceil Bias}}
  MaterialExpressionScalarParameter_18 {class=MaterialExpressionScalarParameter, properties {defaultValue=7, parameterName=Mask Brightness}}
  MaterialExpressionScalarParameter_19 {class=MaterialExpressionScalarParameter, properties {defaultValue=8, parameterName=Ceil Brightness}}
  MaterialExpressionScalarParameter_20 {class=MaterialExpressionScalarParameter, properties {parameterName=Style Lerp}}
  MaterialExpressionScalarParameter_21 {class=MaterialExpressionScalarParameter, properties {defaultValue=-0.03, parameterName=Slide}}
  MaterialExpressionScalarParameter_22 {class=MaterialExpressionScalarParameter, properties {defaultValue=0.5, parameterName=HueShiftSpeed}}
  MaterialExpressionScalarParameter_23 {class=MaterialExpressionScalarParameter, properties {defaultValue=1.2, parameterName=HueShift_Fixed}}
  MaterialExpressionScalarParameter_24 {class=MaterialExpressionScalarParameter, properties {defaultValue=1, parameterName=HueShift_Amount}}
  MaterialExpressionScalarParameter_25 {class=MaterialExpressionScalarParameter, properties {defaultValue=250, parameterName=NearCamFade_Distance}}
  MaterialExpressionScalarParameter_26 {class=MaterialExpressionScalarParameter, properties {defaultValue=1, parameterName=Ray Animation Speed}}
  MaterialExpressionScalarParameter_27 {class=MaterialExpressionScalarParameter, properties {defaultValue=1, parameterName=Ray Pattern Frequency}}
  MaterialExpressionScalarParameter_28 {class=MaterialExpressionScalarParameter, properties {defaultValue=1, parameterName=Brightness}}
  MaterialExpressionScalarParameter_29 {
   class=MaterialExpressionScalarParameter
   properties {controlType=Enumeration, enumeration=/DealershipSharedLibrary/FX/Materials/Enums/Enum_DLR_TrueFalse, parameterName=UseTextureAlphaForOpacity}
  }
  MaterialExpressionScalarParameter_46 {class=MaterialExpressionScalarParameter, properties {defaultValue=1, parameterName=Ray Animation Speed}}
  MaterialExpressionScalarParameter_47 {class=MaterialExpressionScalarParameter, properties {defaultValue=1, parameterName=Ray Animation Speed}}
  MaterialExpressionScalarParameter_48 {class=MaterialExpressionScalarParameter, properties {defaultValue=1, parameterName=Ray Pattern Frequency}}
  MaterialExpressionStaticSwitchParameter_10 {
   class=MaterialExpressionStaticSwitchParameter
   inputs [
    True {from=MaterialExpressionMultiply_20, output=""}
    False {from=MaterialExpressionMaterialFunctionCall_8, output=Result}
   ]
   properties {parameterName=NotDiscoLights}
  }
  MaterialExpressionStaticSwitchParameter_11 {
   class=MaterialExpressionStaticSwitchParameter
   inputs [
    True {from=MaterialExpressionFeatureLevelSwitch_1, output=""}
    False {from=MaterialExpressionQualitySwitch_0, output=""}
   ]
   properties {parameterName=Use_MobilePerf_Settings}
  }
  MaterialExpressionStaticSwitchParameter_12 {
   class=MaterialExpressionStaticSwitchParameter
   inputs [
    True {from=MaterialExpressionMultiply_31, output=""}
    False {from=MaterialExpressionReroute_22, output=""}
   ]
   properties {parameterName=Use_ColorRange_Instead_of_RandomColor?}
  }
  MaterialExpressionStaticSwitchParameter_13 {
   class=MaterialExpressionStaticSwitchParameter
   inputs [
    True {from=MaterialExpressionScalarParameter_23, output=""}
    False {from=MaterialExpressionMultiply_32, output=""}
   ]
   properties {parameterName=Use_Fixed_HueShift_Value?}
  }
  MaterialExpressionStaticSwitchParameter_9 {
   class=MaterialExpressionStaticSwitchParameter
   inputs [
    True {from=MaterialExpressionMultiply_33, output=""}
    False {from=MaterialExpressionReroute_17, output=""}
   ]
   properties {parameterName=Use_NearCamFade?}
  }
  MaterialExpressionSubtract_2 {
   class=MaterialExpressionSubtract
   inputs [
    A {from=MaterialExpressionTextureCoordinate_2}
    B {from=""}
   ]
   properties {constB=0.5}
  }
  MaterialExpressionSubtract_3 {
   class=MaterialExpressionSubtract
   inputs [
    A {from=MaterialExpressionTextureCoordinate_3}
    B {from=""}
   ]
   properties {constB=0.5}
  }
  MaterialExpressionSubtract_4 {
   class=MaterialExpressionSubtract
   inputs [
    A {from=MaterialExpressionTextureCoordinate_4}
    B {from=""}
   ]
   properties {constB=0.5}
  }
  MaterialExpressionSubtract_5 {
   class=MaterialExpressionSubtract
   inputs [
    A {from=MaterialExpressionTextureCoordinate_5}
    B {from=""}
   ]
   properties {constB=0.5}
  }
  MaterialExpressionSubtract_6 {
   class=MaterialExpressionSubtract
   inputs [
    A {from=MaterialExpressionTextureCoordinate_6}
    B {from=""}
   ]
   properties {constB=0.5}
  }
  MaterialExpressionTextureCoordinate_2 {class=MaterialExpressionTextureCoordinate}
  MaterialExpressionTextureCoordinate_3 {class=MaterialExpressionTextureCoordinate}
  MaterialExpressionTextureCoordinate_4 {class=MaterialExpressionTextureCoordinate}
  MaterialExpressionTextureCoordinate_5 {class=MaterialExpressionTextureCoordinate}
  MaterialExpressionTextureCoordinate_6 {class=MaterialExpressionTextureCoordinate}
  MaterialExpressionTextureObjectParameter_2 {
   class=MaterialExpressionTextureObjectParameter
   inputs [
    UVs {from="", output=""}
    Apply View MipBias {from="", output=""}
   ]
   properties {parameterName="Lines Texture (RG)", texture=/Game/Effects/Fort_Effects/Textures/Outlander/T_InTheZone_LightRays_01}
  }
  MaterialExpressionTextureObjectParameter_3 {
   class=MaterialExpressionTextureObjectParameter
   inputs [
    UVs {from="", output=""}
    Apply View MipBias {from="", output=""}
   ]
   properties {parameterName="Lines Texture Mask (R)", texture=&T_NoisePack_01}
  }
  MaterialExpressionTextureObjectParameter_6 {
   class=MaterialExpressionTextureObjectParameter
   inputs [
    UVs {from="", output=""}
    Apply View MipBias {from="", output=""}
   ]
   properties {parameterName="Lines Texture (RG)", texture=&T_InTheZone_LightRays_01}
  }
  MaterialExpressionTextureSample_2 {
   class=MaterialExpressionTextureSample
   inputs [
    UVs {from=MaterialExpressionMultiply_46}
    Tex {from=MaterialExpressionTextureObjectParameter_6}
    Apply View MipBias {from=""}
   ]
   properties {texture=&T_InTheZone_LightRays_01}
  }
  MaterialExpressionTextureSample_3 {
   class=MaterialExpressionTextureSample
   inputs [
    UVs {from=MaterialExpressionMultiply_43}
    Tex {from=MaterialExpressionTextureObjectParameter_2}
    Apply View MipBias {from=""}
   ]
   properties {texture=&T_InTheZone_LightRays_01}
  }
  MaterialExpressionTextureSample_6 {
   class=MaterialExpressionTextureSample
   inputs [
    UVs {from=MaterialExpressionMultiply_69}
    Tex {from=MaterialExpressionTextureObjectParameter_3}
    Apply View MipBias {from=""}
   ]
   properties {texture=&T_NoisePack_01}
  }
  MaterialExpressionTextureSample_7 {
   class=MaterialExpressionTextureSample
   inputs [
    UVs {from=MaterialExpressionMultiply_71}
    Tex {from=""}
    Apply View MipBias {from=""}
   ]
   properties {texture=&T_NoisePack_01}
  }
  MaterialExpressionTime_0 {class=MaterialExpressionTime}
  MaterialExpressionTime_2 {class=MaterialExpressionTime}
  MaterialExpressionTime_3 {class=MaterialExpressionTime}
  MaterialExpressionTime_7 {class=MaterialExpressionTime}
  MaterialExpressionTime_8 {class=MaterialExpressionTime}
  MaterialExpressionVectorParameter_2 {class=MaterialExpressionVectorParameter, properties {parameterName=Color_01, defaultValue=<rgba 1 0 0.1327 1>}}
  MaterialExpressionVectorParameter_3 {class=MaterialExpressionVectorParameter, properties {parameterName=Color_02, defaultValue=<rgba 0 0 1 1>}}
 }
 properties {
  bUsedWithMeshParticles=true
  bUsedWithNanite=true
  bUsedWithNiagaraMeshParticles=true
  bUsedWithNiagaraSprites=true
  bUsedWithParticleSprites=true
  blendMode=BLEND_Additive
  pixelDepthOffsetMode=PDOM_Legacy
  shadingModel=MSM_Unlit
  usedShadingModels=Unlit
 }
}
