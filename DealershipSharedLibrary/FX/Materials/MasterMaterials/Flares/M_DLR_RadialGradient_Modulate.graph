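# Unlit Modulate-blend master material for Niagara sprite flares, drawing a procedural radial gradient that tints what lies behind it.
# A centered soft SphereMask on UVs, raised to an exponent, combines with particle alpha, near-camera fade and depth fade.
# That mask blends from neutral white to particle color, adjusted by multiply/add parameters and optional time-of-day tinting.
# Dynamic parameter index 1 supplies exponent and both fade distances per particle.
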Material /DealershipSharedLibrary/FX/Materials/MasterMaterials/Flares/M_DLR_RadialGradient_Modulate
graph {
 properties {
  bAutomaticallySetUsageInEditor=false
  bUsedWithNiagaraSprites=true
  bUsedWithStaticMesh=false
  blendMode=BLEND_Modulate
  shadingModel=MSM_Unlit
  usedShadingModels=Unlit
 }
 nodes [
  n1 = MaterialFunctionCall(
   Color_Morning=VectorParameter() {group=Time Of Day, parameterName=Color Morning Multiplier, defaultValue=<rgba 0.9896 0.268 0.4929 0>}.RGB
   Color_Day=VectorParameter() {group=Time Of Day, parameterName=Color Day Multiplier, defaultValue=<rgba 1 1 1 0>}.RGB
   Color_Evening=VectorParameter() {group=Time Of Day, parameterName=Color Evening Multiplier, defaultValue=<rgba 0.599 0.2901 0.3358 0>}.RGB
   Color_Night=VectorParameter() {group=Time Of Day, parameterName=Color Night Multiplier, defaultValue=<rgba 0.3498 0.4864 0.6458 0>}.RGB
  ) {materialFunction=&Color_Time_of_Day}
  n2 = ParticleColor()
  n3 = Multiply(A=LinearInterpolate(B=Add(A=@n1.Result), Alpha=ScalarParameter() {parameterName="TimeOfDay_EmissionControl (0-1)"}) {constA=1}, B=@n2.RGB)
  n4 = ScalarParameter() {defaultValue=1, parameterName=ParticleColor-Mult}
  n5 = Constant() {r=0.5}
  n6 = DynamicParameter() {parameterIndex=1, defaultValue=<rgba 1 150 100 1>, paramNames [Exponent, Camera Fade, Depht Fade, Param4]}
  n7 = Power(Base=SphereMask(A=TextureCoordinate(), B=@n5, Radius=@n5) {hardnessPercent=0}, Exp=@n6.Exponent)
  n8 = DynamicParameter() {parameterIndex=1, defaultValue=<rgba 1 150 100 1>, paramNames [Exponent, Camera Fade, Depht Fade, Param4]}
  n9 = MaterialFunctionCall(
   Opacity In=MaterialFunctionCall(Distance=@n8.Camera Fade) {materialFunction=&Near_Camera_Fade}.Result
   DepthFade Distance=@n8.Depht Fade
  ) {materialFunction=&MF_VCVFX_DepthFade}
  LinearInterpolate(
   B=Add(A=Multiply(A=@n3, B=@n4), B=ScalarParameter() {parameterName=ParticleColor-Add})
   Alpha=Multiply(A=@n2.A, B=Multiply(A=@n7, B=@n9.Result))
  ) {constA=1}
 ]
}
